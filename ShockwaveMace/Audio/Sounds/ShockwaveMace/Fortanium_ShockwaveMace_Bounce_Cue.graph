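# Sound cue for the Shockwave Mace bounce, layering a Shockwave Grenade explosion under a Fortanium energy launch sound.
# The explosion varies slightly in pitch, while the louder energy layer is pitched up to a fixed double rate.
# Each layer's quality-level node picks between larger and smaller random wave pools.
# The mix feeds a slam attenuation node, with concurrency limited to a few instances, stopping the oldest.
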SoundCue /ShockwaveMace/Audio/Sounds/ShockwaveMace/Fortanium_ShockwaveMace_Bounce_Cue
root=Mixer_2
nodes [
 {
  id=Attenuation_1
  children [Mixer_1]
  properties {attenuationSettings=/Game/Sounds/Attenuation/Fort_Weapon_Melee_Attack_Special_AOE_Slam_Att}
 }
 {id=Mixer_1, children [Modulator_0,Modulator_1], properties {inputVolume [0.8,1.5]}}
 {id=Mixer_2, children [Attenuation_1, null], properties {inputVolume [1,1]}}
 {id=Modulator_0, children [QualityLevel_5], properties {pitchMax=1.1, pitchMin=0.9}}
 {id=Modulator_1, children [QualityLevel_2], properties {pitchMax=2, pitchMin=2}}
 {id=QualityLevel_2, children [Random_2,Random_2,Random_1,Random_1]}
 {id=QualityLevel_5, children [Random_0,Random_0,Random_7,Random_7]}
 {id=Random_0, children [WavePlayer_0,WavePlayer_1,WavePlayer_2], properties {weights [1,1,1]}}
 {id=Random_1, children [WavePlayer_5, WavePlayer_11], properties {weights [1,1]}}
 {id=Random_2, children [WavePlayer_3, WavePlayer_4, WavePlayer_5, WavePlayer_11], properties {weights [1,1,1,1]}}
 {id=Random_7, children [WavePlayer_1,WavePlayer_2], properties {weights [1,1]}}
 {id=WavePlayer_0, properties {soundWaveAssetPtr=/Game/Sounds/Fort_GamePlay_Sounds/Explosives/ShockwaveGrenade/Grenade_Shockwave_Explode01}}
 {id=WavePlayer_1, properties {soundWaveAssetPtr=/Game/Sounds/Fort_GamePlay_Sounds/Explosives/ShockwaveGrenade/Grenade_Shockwave_Explode02}}
 {id=WavePlayer_11, properties {soundWaveAssetPtr=&/ShockwaveMace_FortaniumEnergy_Launch_04}}
 {id=WavePlayer_2, properties {soundWaveAssetPtr=/Game/Sounds/Fort_GamePlay_Sounds/Explosives/ShockwaveGrenade/Grenade_Shockwave_Explode03}}
 {id=WavePlayer_3, properties {soundWaveAssetPtr=&/ShockwaveMace_FortaniumEnergy_Launch_01}}
 {id=WavePlayer_4, properties {soundWaveAssetPtr=&/ShockwaveMace_FortaniumEnergy_Launch_02}}
 {id=WavePlayer_5, properties {soundWaveAssetPtr=&/ShockwaveMace_FortaniumEnergy_Launch_03}}
]
properties {
 bOverrideConcurrency=true
 maxDistance=13031.375
 soundClassObject=&Fort_Gadgets_Items
 volumeMultiplier=0.9
 attenuationOverrides {distanceAlgorithm=NaturalSound, falloffDistance=3800, lPFRadiusMax=3000, lPFRadiusMin=2000, attenuationShapeExtents {x=2200}}
 concurrencyOverrides {maxCount=2, resolutionRule=StopOldest}
}
