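# Sound cue for the Shockwave Mace slam, blending mace hit sounds with reused Shockwave Grenade explosions.
# A top mixer combines a pitch-randomized hit and explosion pair under melee attack attenuation with volume-randomized distant explosions under close-range stereo attenuation.
# Quality-level nodes pick from random wave pools, with fewer variants in the latter two slots.
# Overridden concurrency caps playback at two instances, stopping the oldest.
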SoundCue /ShockwaveMace/Audio/Sounds/ShockwaveMace/Fortanium_ShockwaveMace_Slam_Cue
root=Mixer_2
nodes [
 {
  id=Attenuation_0
  children [Modulator_13]
  properties {attenuationSettings=/Game/Sounds/Attenuation/Fort_Weapon_Melee_Attack_Close_Range_Stereo_Att}
 }
 {id=Attenuation_1, children [Mixer_1], properties {attenuationSettings=/ShockwaveMace/Audio/Settings/Fort_Weapon_Melee_Attack_Att}}
 {id=Mixer_1, children [Modulator_0,Modulator_1], properties {inputVolume [0.8,1]}}
 {id=Mixer_2, children [Attenuation_1, Attenuation_0], properties {inputVolume [1,1]}}
 {id=Modulator_0, children [QualityLevel_5], properties {pitchMax=1.1, pitchMin=0.9}}
 {id=Modulator_1, children [QualityLevel_2], properties {pitchMax=1.1, pitchMin=0.9}}
 {id=Modulator_13, children [QualityLevel_4], properties {volumeMax=1.1, volumeMin=1}}
 {id=QualityLevel_2, children [Random_2,Random_2,Random_1,Random_1]}
 {id=QualityLevel_4, children [Random_8,Random_8,Random_5,Random_5]}
 {id=QualityLevel_5, children [Random_0,Random_0,Random_7,Random_7]}
 {id=Random_0, children [WavePlayer_0,WavePlayer_1,WavePlayer_2], properties {weights [1,1,1]}}
 {id=Random_1, children [WavePlayer_17, WavePlayer_18, WavePlayer_19], properties {weights [1,1,1]}}
 {
  id=Random_2
  children [WavePlayer_6, WavePlayer_7, WavePlayer_8, WavePlayer_9, WavePlayer_10, WavePlayer_17, WavePlayer_18, WavePlayer_19]
  properties {weights [1,1,1,1,1,1,1,1]}
 }
 {id=Random_5, children [WavePlayer_15, WavePlayer_16], properties {weights [1,1]}}
 {id=Random_7, children [WavePlayer_1,WavePlayer_2], properties {weights [1,1]}}
 {id=Random_8, children [WavePlayer_14, WavePlayer_15, WavePlayer_16], properties {weights [1,1,1]}}
 {id=WavePlayer_0, properties {soundWaveAssetPtr=/Game/Sounds/Fort_GamePlay_Sounds/Explosives/ShockwaveGrenade/Grenade_Shockwave_Explode01}}
 {id=WavePlayer_1, properties {soundWaveAssetPtr=/Game/Sounds/Fort_GamePlay_Sounds/Explosives/ShockwaveGrenade/Grenade_Shockwave_Explode02}}
 {id=WavePlayer_10, properties {soundWaveAssetPtr=&/ShockwaveMace_Hit_05}}
 {
  id=WavePlayer_14
  properties {soundWaveAssetPtr=/Game/Sounds/Fort_GamePlay_Sounds/Explosives/ShockwaveGrenade/Grenade_Shockwave_Explode_Distant01}
 }
 {
  id=WavePlayer_15
  properties {soundWaveAssetPtr=/Game/Sounds/Fort_GamePlay_Sounds/Explosives/ShockwaveGrenade/Grenade_Shockwave_Explode_Distant02}
 }
 {
  id=WavePlayer_16
  properties {soundWaveAssetPtr=/Game/Sounds/Fort_GamePlay_Sounds/Explosives/ShockwaveGrenade/Grenade_Shockwave_Explode_Distant03}
 }
 {id=WavePlayer_17, properties {soundWaveAssetPtr=&/ShockwaveMace_Hit_06}}
 {id=WavePlayer_18, properties {soundWaveAssetPtr=&/ShockwaveMace_Hit_07}}
 {id=WavePlayer_19, properties {soundWaveAssetPtr=&/ShockwaveMace_Hit_08}}
 {id=WavePlayer_2, properties {soundWaveAssetPtr=/Game/Sounds/Fort_GamePlay_Sounds/Explosives/ShockwaveGrenade/Grenade_Shockwave_Explode03}}
 {id=WavePlayer_6, properties {soundWaveAssetPtr=&/ShockwaveMace_Hit_01}}
 {id=WavePlayer_7, properties {soundWaveAssetPtr=&/ShockwaveMace_Hit_02}}
 {id=WavePlayer_8, properties {soundWaveAssetPtr=&/ShockwaveMace_Hit_03}}
 {id=WavePlayer_9, properties {soundWaveAssetPtr=&/ShockwaveMace_Hit_04}}
]
properties {
 bOverrideConcurrency=true
 maxDistance=10058.75
 soundClassObject=&Fort_Gadgets_Items
 volumeMultiplier=1.1
 attenuationOverrides {distanceAlgorithm=NaturalSound, falloffDistance=3800, lPFRadiusMax=3000, lPFRadiusMin=2000, attenuationShapeExtents {x=2200}}
 concurrencyOverrides {maxCount=2, resolutionRule=StopOldest}
}
