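# Sound cue for the attack swing of the Shockwave Mace melee weapon.
# A Random node picks one of three swing waves, weighting the third twice as heavily as each of the others, and feeds a root Modulator.
# The cue plays at reduced volume in the melee weapon sound class, referencing a close-range stereo melee attenuation asset with attenuation override enabled.
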SoundCue /ShockwaveMace/Audio/Sounds/ShockwaveMace/ShockwaveMace_AttackSwing_Cue
root=Modulator_0
nodes [
 {id=Modulator_0, children [Random_0]}
 {id=Random_0, children [WavePlayer_0,WavePlayer_1,WavePlayer_2], properties {weights [1,1,2]}}
 {id=WavePlayer_0, properties {soundWaveAssetPtr=&/ShockwaveMace_AttackSwing_01}}
 {id=WavePlayer_1, properties {soundWaveAssetPtr=&/ShockwaveMace_AttackSwing_02}}
 {id=WavePlayer_2, properties {soundWaveAssetPtr=&/ShockwaveMace_AttackSwing_03}}
]
properties {
 attenuationSettings=/Game/Sounds/Attenuation/Fort_Weapon_Melee_Attack_Close_Range_Stereo_Att
 bOverrideAttenuation=true
 maxDistance=4000
 soundClassObject=&Fort_WeaponMelee
 volumeMultiplier=0.8
 concurrencySet [null]
}
