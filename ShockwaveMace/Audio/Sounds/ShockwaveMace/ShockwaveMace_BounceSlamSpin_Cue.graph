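# Sound cue for the Shockwave Mace's bounce slam spin sound.
# A Random node picks one of nine equally weighted wave variations and feeds it through a Modulator at the root for pitch and volume variation.
# It uses the melee weapon sound class and a melee attack attenuation asset for spatial falloff.
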SoundCue /ShockwaveMace/Audio/Sounds/ShockwaveMace/ShockwaveMace_BounceSlamSpin_Cue
root=Modulator_0
nodes [
 {id=Modulator_0, children [Random_0]}
 {
  id=Random_0
  children [WavePlayer_0,WavePlayer_1,WavePlayer_2,WavePlayer_3,WavePlayer_4,WavePlayer_5,WavePlayer_6,WavePlayer_7,WavePlayer_8]
  properties {weights [1,1,1,1,1,1,1,1,1]}
 }
 {id=WavePlayer_0, properties {soundWaveAssetPtr=&/ShockwaveMace_BounceSlamSpin_01}}
 {id=WavePlayer_1, properties {soundWaveAssetPtr=&/ShockwaveMace_BounceSlamSpin_02}}
 {id=WavePlayer_2, properties {soundWaveAssetPtr=&/ShockwaveMace_BounceSlamSpin_03}}
 {id=WavePlayer_3, properties {soundWaveAssetPtr=&/ShockwaveMace_BounceSlamSpin_04}}
 {id=WavePlayer_4, properties {soundWaveAssetPtr=&/ShockwaveMace_BounceSlamSpin_05}}
 {id=WavePlayer_5, properties {soundWaveAssetPtr=&/ShockwaveMace_BounceSlamSpin_06}}
 {id=WavePlayer_6, properties {soundWaveAssetPtr=&/ShockwaveMace_BounceSlamSpin_07}}
 {id=WavePlayer_7, properties {soundWaveAssetPtr=&/ShockwaveMace_BounceSlamSpin_08}}
 {id=WavePlayer_8, properties {soundWaveAssetPtr=&/ShockwaveMace_BounceSlamSpin_09}}
]
properties {
 attenuationSettings=/ShockwaveMace/Audio/Settings/Fort_Weapon_Melee_Attack_Att
 maxDistance=4000
 soundClassObject=&Fort_WeaponMelee
 volumeMultiplier=1
}
