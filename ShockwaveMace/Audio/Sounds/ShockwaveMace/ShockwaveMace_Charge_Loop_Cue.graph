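# Sound cue for the Shockwave Mace's looping charge sound, classed as melee weapon audio.
# A Switch on the PowerLevel parameter leaves index 0 silent and routes indices 1-5 through Modulators to looping IdleLoop waves, with index 3 reusing Level_02.
# Pitch varies by level, but only level 1 is audible because the other Modulators have zero volume.
# Attenuation is overridden on the cue, and concurrency allows one instance.
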SoundCue /ShockwaveMace/Audio/Sounds/ShockwaveMace/ShockwaveMace_Charge_Loop_Cue
root=Switch_0
nodes [
 {id=Modulator_2, children [WavePlayer_6], properties {pitchMax=0.8, pitchMin=0.8, volumeMax=0, volumeMin=0}}
 {id=Modulator_3, children [WavePlayer_0], properties {pitchMax=0.9, pitchMin=0.9, volumeMax=0, volumeMin=0}}
 {id=Modulator_4, children [WavePlayer_8], properties {pitchMax=1.2, pitchMin=1.2, volumeMax=0, volumeMin=0}}
 {id=Modulator_5, children [WavePlayer_4], properties {pitchMax=1, pitchMin=1, volumeMax=0, volumeMin=0}}
 {id=Modulator_6, children [WavePlayer_5], properties {pitchMax=1, pitchMin=1, volumeMax=2.2, volumeMin=2.2}}
 {id=Switch_0, children [null,Modulator_6,Modulator_2,Modulator_3,Modulator_5,Modulator_4], properties {intParameterName=PowerLevel}}
 {
  id=WavePlayer_0
  properties {bLooping=true, soundWaveAssetPtr=&/Fortanium_ShockwaveMace_IdleLoop_Level_02}
 }
 {
  id=WavePlayer_4
  properties {bLooping=true, soundWaveAssetPtr=&/Fortanium_ShockwaveMace_IdleLoop_Level_04}
 }
 {
  id=WavePlayer_5
  properties {bLooping=true, soundWaveAssetPtr=&/Fortanium_ShockwaveMace_IdleLoop_Level_01}
 }
 {
  id=WavePlayer_6
  properties {bLooping=true, soundWaveAssetPtr=&Fortanium_ShockwaveMace_IdleLoop_Level_02}
 }
 {
  id=WavePlayer_8
  properties {bLooping=true, soundWaveAssetPtr=&/Fortanium_ShockwaveMace_IdleLoop_Level_05}
 }
]
properties {
 attenuationSettings=/Game/Sounds/Creative/Attenuations/Fort_Creative_SmallToy_Att
 bOverrideAttenuation=true
 bOverrideConcurrency=true
 maxDistance=1600
 pitchMultiplier=0.9
 soundClassObject=&Fort_WeaponMelee
 volumeMultiplier=1
 attenuationOverrides {falloffDistance=1500, stereoSpread=100, attenuationShapeExtents {x=100}}
 concurrencyOverrides {maxCount=1}
}
