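# Sound cue for the Shockwave Mace FlyLand sound, classed as a melee weapon sound.
# A random node picks one of four FlyLand wave variants, with the third weighted twice as heavily as the others.
# The pick passes through delay and modulator nodes left at default settings.
# Playback references the shared close-range stereo melee attack attenuation, with the attenuation override flag enabled.
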SoundCue /ShockwaveMace/Audio/Sounds/ShockwaveMace/ShockwaveMace_FlyLand_Cue
root=Modulator_0
nodes [
 {id=Delay_1, children [Random_0]}
 {id=Modulator_0, children [Delay_1]}
 {id=Random_0, children [WavePlayer_3,WavePlayer_4,WavePlayer_5,WavePlayer_7], properties {weights [1,1,2,1]}}
 {id=WavePlayer_3, properties {soundWaveAssetPtr=&/Fortanium_ShockwaveMace_FlyLand_01}}
 {id=WavePlayer_4, properties {soundWaveAssetPtr=&/Fortanium_ShockwaveMace_FlyLand_02}}
 {id=WavePlayer_5, properties {soundWaveAssetPtr=&/Fortanium_ShockwaveMace_FlyLand_03}}
 {id=WavePlayer_7, properties {soundWaveAssetPtr=&/Fortanium_ShockwaveMace_FlyLand_04}}
]
properties {
 attenuationSettings=/Game/Sounds/Attenuation/Fort_Weapon_Melee_Attack_Close_Range_Stereo_Att
 bOverrideAttenuation=true
 maxDistance=4000
 soundClassObject=&Fort_WeaponMelee
 volumeMultiplier=1
 concurrencySet [null]
}
