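# Sound cue for the Shockwave Mace's charge-gain sound, varying with the mace's power level.
# A switch on the PowerLevel parameter leaves its first two slots silent, then plays a random first-charge variant or one of three higher-charge waves.
# Fixed modulators set volume and pitch per stage, with the top stage pitched up.
# It uses the melee weapon sound class, small-toy attenuation with overrides, and single-instance concurrency.
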SoundCue /ShockwaveMace/Audio/Sounds/ShockwaveMace/ShockwaveMace_GainCharge_Cue
root=Switch_0
nodes [
 {id=Modulator_1, children [Random_0], properties {volumeMax=0.8, volumeMin=0.8}}
 {id=Modulator_3, children [WavePlayer_0], properties {pitchMax=0.9, pitchMin=0.9, volumeMax=1, volumeMin=1}}
 {id=Modulator_4, children [WavePlayer_4], properties {pitchMax=1.2, pitchMin=1.2, volumeMax=0.7, volumeMin=0.7}}
 {id=Modulator_5, children [WavePlayer_1], properties {pitchMax=1, pitchMin=1, volumeMax=1, volumeMin=1}}
 {id=Random_0, children [WavePlayer_3, WavePlayer_9, WavePlayer_10], properties {weights [1,1,1]}}
 {id=Switch_0, children [null,null,Modulator_1,Modulator_3,Modulator_5,Modulator_4], properties {intParameterName=PowerLevel}}
 {id=WavePlayer_0, properties {soundWaveAssetPtr=&/Fortanium_ShockwaveMace_GainCharge_02}}
 {id=WavePlayer_1, properties {soundWaveAssetPtr=&/Fortanium_ShockwaveMace_GainCharge_03}}
 {id=WavePlayer_10, properties {soundWaveAssetPtr=&/Fortanium_ShockwaveMace_GainCharge_1C}}
 {id=WavePlayer_3, properties {soundWaveAssetPtr=&/Fortanium_ShockwaveMace_GainCharge_1A}}
 {id=WavePlayer_4, properties {soundWaveAssetPtr=&/Fortanium_ShockwaveMace_GainCharge_04}}
 {id=WavePlayer_9, properties {soundWaveAssetPtr=&/Fortanium_ShockwaveMace_GainCharge_1B}}
]
properties {
 attenuationSettings=/Game/Sounds/Creative/Attenuations/Fort_Creative_SmallToy_Att
 bOverrideAttenuation=true
 bOverrideConcurrency=true
 maxDistance=1600
 pitchMultiplier=0.9
 soundClassObject=&Fort_WeaponMelee
 volumeMultiplier=2
 attenuationOverrides {bEnableReverbSend=false, bEnableSendToAudioLink=false, falloffDistance=1500, stereoSpread=100, attenuationShapeExtents {x=100}}
 concurrencyOverrides {maxCount=1}
}
