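# Sound cue for the Shockwave Mace's in-air loop, using the glider in-air sound class.
# A mixer layers one of three looping in-air waves, picked at equal-weight random, with a level-5 idle-loop wave not flagged to loop, each through its own modulator.
# The mix feeds an attenuation node using the melee AOE slam settings.
# The cue overrides concurrency to cap simultaneous instances and boosts volume.
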SoundCue /ShockwaveMace/Audio/Sounds/ShockwaveMace/ShockwaveMace_InAirLoop_Cue
root=Attenuation_1
nodes [
 {
  id=Attenuation_1
  children [Mixer_0]
  properties {attenuationSettings=/Game/Sounds/Attenuation/Fort_Weapon_Melee_Attack_Special_AOE_Slam_Att}
 }
 {id=Mixer_0, children [Modulator_2,Modulator_0], properties {inputVolume [1,1]}}
 {id=Modulator_0, children [WavePlayer_2]}
 {id=Modulator_2, children [Random_0]}
 {id=Random_0, children [WavePlayer_0,WavePlayer_1,WavePlayer_3], properties {weights [1,1,1]}}
 {
  id=WavePlayer_0
  properties {bLooping=true, soundWaveAssetPtr=&/Fortanium_ShockwaveMace_InAirLoop_04}
 }
 {
  id=WavePlayer_1
  properties {bLooping=true, soundWaveAssetPtr=&/Fortanium_ShockwaveMace_InAirLoop_01}
 }
 {id=WavePlayer_2, properties {soundWaveAssetPtr=&/Fortanium_ShockwaveMace_IdleLoop_Level_05}}
 {
  id=WavePlayer_3
  properties {bLooping=true, soundWaveAssetPtr=&/Fortanium_ShockwaveMace_InAirLoop_03}
 }
]
properties {
 bOverrideConcurrency=true
 maxDistance=13031.375
 soundClassObject=&Fort_Gliders_InAir
 volumeMultiplier=1.5
 attenuationOverrides {bAttenuate=false, falloffDistance=100000}
 concurrencyOverrides {maxCount=4}
}
