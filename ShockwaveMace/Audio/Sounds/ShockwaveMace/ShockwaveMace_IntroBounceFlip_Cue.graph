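# Sound cue for the Shockwave Mace intro bounce flip sound, classed as melee weapon audio.
# A Random node picks one of three bounce flip wave variations with equal weight and feeds it into a Modulator left at default settings.
# The cue uses the melee weapon sound class and a shared melee attack attenuation asset for spatial falloff.
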SoundCue /ShockwaveMace/Audio/Sounds/ShockwaveMace/ShockwaveMace_IntroBounceFlip_Cue
root=Modulator_0
nodes [
 {id=Modulator_0, children [Random_0]}
 {id=Random_0, children [WavePlayer_0,WavePlayer_1,WavePlayer_2], properties {weights [1,1,1]}}
 {id=WavePlayer_0, properties {soundWaveAssetPtr=&/ShockwaveMace_IntroBounceFlip_01}}
 {id=WavePlayer_1, properties {soundWaveAssetPtr=&/ShockwaveMace_IntroBounceFlip_02}}
 {id=WavePlayer_2, properties {soundWaveAssetPtr=&/ShockwaveMace_IntroBounceFlip_03}}
]
properties {
 attenuationSettings=/ShockwaveMace/Audio/Settings/Fort_Weapon_Melee_Attack_Att
 maxDistance=4000
 soundClassObject=&Fort_WeaponMelee
 volumeMultiplier=1
}
