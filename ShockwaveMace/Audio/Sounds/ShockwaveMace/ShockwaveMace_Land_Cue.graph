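# Sound cue that plays a landing sound for the Shockwave Mace.
# A Random node picks one of three equally weighted Land wave variations and feeds it into a root Modulator left at its default settings.
# The cue plays at reduced volume in the melee weapon sound class and uses the shared weapon-drop attenuation asset.
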SoundCue /ShockwaveMace/Audio/Sounds/ShockwaveMace/ShockwaveMace_Land_Cue
root=Modulator_0
nodes [
 {id=Modulator_0, children [Random_0]}
 {id=Random_0, children [WavePlayer_3,WavePlayer_4,WavePlayer_5], properties {weights [1,1,1]}}
 {id=WavePlayer_3, properties {soundWaveAssetPtr=&/ShockwaveMace_Land_01}}
 {id=WavePlayer_4, properties {soundWaveAssetPtr=&/ShockwaveMace_Land_02}}
 {id=WavePlayer_5, properties {soundWaveAssetPtr=&/ShockwaveMace_Land_03}}
]
properties {
 attenuationSettings=&Fort_Weapon_Drop_Att
 maxDistance=3000
 soundClassObject=&Fort_WeaponMelee
 volumeMultiplier=0.6
}
