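# Sound cue for the Shockwave Mace launch whoosh, classed as melee weapon audio.
# A Random node picks one of four whoosh waves, weighted toward the third, and feeds through a Delay left at default settings into a root Modulator for slight variation.
# It uses the shared melee special AOE slam attenuation and a slightly boosted volume.
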SoundCue /ShockwaveMace/Audio/Sounds/ShockwaveMace/ShockwaveMace_LaunchWhoosh_Cue
root=Modulator_0
nodes [
 {id=Delay_1, children [Random_0]}
 {id=Modulator_0, children [Delay_1]}
 {id=Random_0, children [WavePlayer_0,WavePlayer_1,WavePlayer_2,WavePlayer_6], properties {weights [1,1,2,1]}}
 {id=WavePlayer_0, properties {soundWaveAssetPtr=&/Fortanium_ShockwaveMace_LaunchWhoosh_01}}
 {id=WavePlayer_1, properties {soundWaveAssetPtr=&/Fortanium_ShockwaveMace_LaunchWhoosh_02}}
 {id=WavePlayer_2, properties {soundWaveAssetPtr=&/Fortanium_ShockwaveMace_LaunchWhoosh_03}}
 {id=WavePlayer_6, properties {soundWaveAssetPtr=&/Fortanium_ShockwaveMace_LaunchWhoosh_04}}
]
properties {
 attenuationSettings=/Game/Sounds/Attenuation/Fort_Weapon_Melee_Attack_Special_AOE_Slam_Att
 maxDistance=13031.375
 soundClassObject=&Fort_WeaponMelee
 volumeMultiplier=1.2
 concurrencySet [null]
}
